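; PlotTwist Rift Gate conversation actor for the Heavy Lamp NPC, offering players Force training.
; Interaction checks tags and distance, then starts the conversation; players with Force powers get a deny message, and training or gate despawn retires the actor.
; Box overlaps drive affiliation-based proximity messages and a timer turning the owning NPC pawn, rather than a hologram, toward the trainee or closest non-bot player.
; Hotfixable ScalableFloats tune distances, delays and rotation.
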(event EventBeginPlay
  (bind _row_readytoshowmessagedelay (Variables|Hotfix|GetRow_ReadyToShowMessageDelay))
  (bind _returnvalue (ScalableFloat|GetValueatLevel _row_readytoshowmessagedelay))
  (|CallParentFunction)
  (Networking|SwitchHasAuthority
    (:Authority
      (Class|FortAsyncActionCurrentPlaylistReady|PlaylistDataReadyAsync
        (:OnReady
          (bind _associatedriftgate (Variables|Default|GetAssociatedRiftGate)
            (:then
              (Default|BindEventtoOnRiftGateForcedDespawn _associatedriftgate (EventDispatchers|CreateEvent :!event Custom|OnRiftGateForcedDespawn)))
            (:else)))))
    (:Remote
      (CallFunction|CheckforPawninArea) ; look at functionality
      ; Handle Ready to show message delay
      (if (> _returnvalue 0)
        (Utilities|FlowControl|Delay _returnvalue) ; We don't want to show proximity message on begin play
        (else
          _))
      (Variables|Dialogue|System|SetbIsReadyToShowProximityMessage true))))

(event Custom|ShowProximityMessage
  (params (FortPlayerPawn :type Fort%20Player%20Pawn%20Object%20Reference))
  (bind _row_notreadymessagedelay (Variables|Hotfix|GetRow_NotReadyMessageDelay))
  (bind _returnvalue (ScalableFloat|GetValueatLevel _row_notreadymessagedelay))
  (bind _row_proximitymessages_mintimebetween (Variables|Dialogue|Hotfixes|GetRow_ProximityMessages_MinTimeBetween))
  (bind _returnvalue_2 (ScalableFloat|GetValueatLevel _row_proximitymessages_mintimebetween))
  ;[1 0 0.75] Show Proximity Message - WIP - uses Hotfix
  (bind _isvalid (Dialogue|UpdateCachedProximityMessage FortPlayerPawn)) ; get a proximity message\nAND checks CanExecuteCosmetic Events
  (if _isvalid
    (Utilities|FlowControl|DoOnce false
      (:Completed
        (if (Variables|Dialogue|System|GetbIsReadyToShowProximityMessage)
          (InteractionUI|ShowMessage (Variables|Default|GetNonPlayerConversationComponent) (Variables|Dialogue|System|GetCachedProximityMessage))
          ;; Handle min time between messages delay
          (if (> _returnvalue_2 0)
            (Utilities|FlowControl|Delay _returnvalue_2)))
        ;; Handle not ready delay
        (if (> _returnvalue 0)
          (Utilities|FlowControl|Delay _returnvalue))))))

(event OnConversationEnded(NonPlayerConversationComponent)
  ; Play an animation if training is not accepted
  (Variables|Default|SetBeingGrantedPowersPawn 0))

(event Custom|SetUIMarkerVisibility
  (params (Visible :type Boolean))
  ;[1 1 1] SetUIMarkerVisibility
  (if Visible
    (InteractionUI|RemoveHideUIIndicatorReason (Variables|Default|GetNonPlayerConversationComponent) (Variables|UI|GetHideUIIndicatorReason))
    (else
      (InteractionUI|AddHideUIIndicatorReason (Variables|Default|GetNonPlayerConversationComponent) (Variables|UI|GetHideUIIndicatorReason)))))

(event Custom|ShowDenyTrainingMessage
  (params (FortPlayerPawn :type Fort%20Player%20Pawn%20Object%20Reference))
  ;[1 0 0.75] Show Proximity Message - WIP - uses Hotfix
  (if (and (Pawn|IsLocallyViewed FortPlayerPawn) (ScalableFloat|GetValueasBool (Variables|Dialogue|Hotfixes|GetRow_ShowDenyTrainingEnabled)))
    (InteractionUI|ShowMessage (Variables|Default|GetNonPlayerConversationComponent) (Variables|Dialogue|GetDenyTrainingMessage))
    (Audio|PlaySound2D (Variables|Sounds|GetDenyTrainingSound) 1.0 1.0 0.0 0 FortPlayerPawn)))

(event Interaction|EventBlueprintOnInteract
  (params (InteractingPawn :type Fort%20Pawn%20Object%20Reference) (InteractionBeingAttempted :type EInteractionBeingAttempted%20Enum :default "FirstInteraction"))
  (if (Interaction|BlueprintCanInteract InteractingPawn InteractionBeingAttempted)
    (if (not (GameplayTags|HasMatchingGameplayTag InteractingPawn (Variables|Interaction|GetTag_HasForcePower)))
      (bind _output_get (Variables|Interaction|SetLastPawnToInteract InteractingPawn))
      (Variables|Default|SetBeingGrantedPowersPawn _output_get)
      (switch Utilities|FlowControl|Switch|SwitchonEInteractionBeingAttempted InteractionBeingAttempted
        (:FirstInteraction
          (bind _asfort_player_controller (Utilities|Casting|CastToFortPlayerController (Pawn|GetController (Variables|Interaction|GetLastPawnToInteract))))
          (Variables|Interaction|SetIsInteracting true)
          (Networking|ForceNetUpdate)
          (Class|PlotTwistRiftGateConversationComponent|StartConversation (Variables|Default|GetPlotTwistRiftGateConversation) _asfort_player_controller)
          (Ability|Tasks|WaitGameplayTagRemoveFromActor (Variables|Interaction|GetLastPawnToInteract) (Variables|Interaction|GetTag_InteractingPawnInConversation)
            (:Removed
              (Variables|Interaction|SetIsInteracting false)
              (Networking|ForceNetUpdate))))))))

(event Custom|OnRiftGateForcedDespawn
  (Default|CallForcedEndOfTraining)
  (Variables|Dialogue|System|SetbIsReadyToShowProximityMessage false)
  (Variables|Interaction|SetbAllowInteract false)
  (Variables|Interaction|SetHasGrantedTraining true)
  (Utilities|Time|ClearandInvalidateTimerbyHandle (Variables|LookAtPlayer|System|GetTimerHandle_FacePawn))
  (Actor|SetLifeSpan 2.5) ; destroy this conversation actor, \nno future conversations are possible
)

(event Custom|GrantTraining
  (Default|CallTrainingAcceptedbyConversation)
  (Variables|Dialogue|System|SetbIsReadyToShowProximityMessage false)
  (Variables|Interaction|SetbAllowInteract false)
  (Variables|Interaction|SetHasGrantedTraining true)
  (Utilities|Time|ClearandInvalidateTimerbyHandle (Variables|LookAtPlayer|System|GetTimerHandle_FacePawn))
  (Actor|SetLifeSpan 2.5) ; destroy this conversation actor, \nno future conversations are possible
)

(event Custom|ForceRemove
  ; Force Remove this actor - We had a case with Vader where we had to teleport him next to the rift if he failed to go there, and this can happen after this actor was spawned and the UI would stay in the world. This allows us to remove thi sactor and spawn a new one at the rightlocaiton.
  (InteractionUI|AddHideUIIndicatorReason (Variables|Default|GetNonPlayerConversationComponent) (Variables|UI|GetHideUIIndicatorReason))
  (LiveEdit|Object|ObjectDestroyed (Variables|Default|GetNonPlayerConversationComponent))
  (Actor|DestroyActor))

(event OnComponentBeginOverlap(LookAtPawnCollisionBox)
  (params (OverlappedComponent :type Primitive%20Component%20Object%20Reference) (OtherActor :type Actor%20Object%20Reference) (OtherComp :type Primitive%20Component%20Object%20Reference) (OtherBodyIndex :type Integer) (bFromSweep :type Boolean) (SweepResult :type Hit%20Result%20Structure%20%28by%20ref%29))
  ;[1 1 0] FacingPawnFunctionality
  (Networking|CanExecuteCosmeticEvents
    (:True
      (CallFunction|CheckforPawninArea)
      (bind _asfort_player_pawn (Utilities|Casting|CastToFortPlayerPawn OtherActor))
      (CallFunction|ShowProximityMessage _asfort_player_pawn))))

(event OnComponentEndOverlap(LookAtPawnCollisionBox)
  (params (OverlappedComponent :type Primitive%20Component%20Object%20Reference) (OtherActor :type Actor%20Object%20Reference) (OtherComp :type Primitive%20Component%20Object%20Reference) (OtherBodyIndex :type Integer))
  ;[1 1 0] FacingPawnFunctionality
  (Networking|CanExecuteCosmeticEvents
    (:True
      (CallFunction|CheckforPawninArea))))

(event Custom|CheckForPawnInArea
  ;[1 1 0] FacingPawnFunctionality
  (bind _closestvalidlookatpawn (LookAtPlayer|GetClosestValidLookatPawn (Transformation|GetActorLocation) (Utilities|Array|MakeArray (Variables|Default|GetOwningNPCPawn))))
  (Utilities|IsValid _closestvalidlookatpawn
    (:Is%20Valid
      (if (<= (Transformation|GetDistanceTo _closestvalidlookatpawn self) (ScalableFloat|GetValueatLevel (Variables|LookAtPlayer|Hotfixes|GetRow_FacePawn_BeginChecksDistance)))
        (bind _returnvalue (Utilities|Time|SetTimerbyEvent (EventDispatchers|CreateEvent :!event Custom|FaceInteractingOrClosestPawn) (Math|Float|SafeDivide 1.0 (ScalableFloat|GetValueatLevel (Variables|LookAtPlayer|Hotfixes|GetRow_FacePawn_UpdateRate))) true))
        (Variables|LookAtPlayer|System|SetTimerHandle_FacePawn _returnvalue)))
    (:Is%20Not%20Valid
      (Utilities|Time|ClearandInvalidateTimerbyHandle (Variables|LookAtPlayer|System|GetTimerHandle_FacePawn)))))

(event Custom|FaceInteractingOrClosestPawn
  (bind _row_facepawn_rotationperupdate (Variables|LookAtPlayer|Hotfixes|GetRow_FacePawn_RotationPerUpdate))
  (bind _returnvalue (ScalableFloat|GetValueatLevel _row_facepawn_rotationperupdate))
  ;[1 1 0] FacingPawnFunctionality
  ;;[0.4656 0.6228 1] Event FaceInteractingOrClosestPawn (OVERRIDE)
  ;; Ignores the OwningNPC (HeavyLamp) and turns the Heavy Lamp pawn instead of the hologram
  (bind _beinggrantedpowerspawn (Variables|Default|GetBeingGrantedPowersPawn)
    (:then
      (Variables|LookAtPlayer|System|SetHologramsFocusedPawn _beinggrantedpowerspawn))
    (:else
      (if (not (Variables|Interaction|GetHasGrantedTraining))
        (bind _closestvalidlookatpawn (LookAtPlayer|GetClosestValidLookatPawn (Transformation|GetActorLocation) (Utilities|Array|MakeArray (Variables|Default|GetOwningNPCPawn))))
        (Utilities|IsValid _closestvalidlookatpawn
          (:Is%20Valid
            (Variables|LookAtPlayer|System|SetHologramsFocusedPawn _closestvalidlookatpawn))) ; can return null
      )))
  (bind _owningnpcpawn (Variables|Default|GetOwningNPCPawn)
    (:then
      (bind _hologramsfocusedpawn (Variables|LookAtPlayer|System|GetHologramsFocusedPawn)
        (:then
          (bind _output_get (Variables|LookAtPlayer|System|SetDesiredLookatYaw (LookAtPlayer|GetLookatYawValue _hologramsfocusedpawn)))
          (bind (_returnvalue_roll _returnvalue_pitch _returnvalue_yaw) (Math|Rotator|BreakRotator (Transformation|GetWorldRotation (Variables|/Script/Engine.Actor|GetRootComponent _owningnpcpawn))))
          (if (not (Math|Float|NearlyEqual(Float) _output_get _returnvalue_yaw (* _returnvalue 1.5)))
            (bind (_returnvalue_roll_2 _returnvalue_pitch_2 _returnvalue_yaw_2) (Math|Rotator|BreakRotator (Transformation|GetWorldRotation (Variables|/Script/Engine.Actor|GetRootComponent _owningnpcpawn))))
            (bind (_returnvalue_roll_3 _returnvalue_pitch_3 _returnvalue_yaw_3) (Math|Rotator|BreakRotator (Math|Interpolation|RInterpTo (Math|Rotator|MakeRotator 0.0 0.0 _returnvalue_yaw_2) (Math|Rotator|MakeRotator 0.0 0.0 _output_get) (Math|Float|SafeDivide 1.0 (ScalableFloat|GetValueatLevel (Variables|LookAtPlayer|Hotfixes|GetRow_FacePawn_UpdateRate))) _returnvalue)))
            (Transformation|SetWorldRotation (Variables|/Script/Engine.Actor|GetRootComponent _owningnpcpawn) (Math|Rotator|MakeRotator 0.0 0.0 _returnvalue_yaw_3) false false)))
        (:else)))
    (:else)))

(fn BlueprintCanInteract
  (params (InteractingPawn :type Fort%20Pawn%20Object%20Reference) (InteractionBeingAttempted :type EInteractionBeingAttempted%20Enum :default "FirstInteraction") (InteractionType :type TInteractionType%20Enum :default "IT_NoInteraction"))
  (if (Utilities|Enum|Equal(Enum) InteractionType "IT_Simple")
    (return (Math|Boolean|ANDBoolean (not (GameplayTags|HasAnyMatchingGameplayTags InteractingPawn (Variables|Interaction|GetTags_InteractingPawnPreventInteract))) (not (Variables|Interaction|GetIsInteracting)) (not (Variables|Interaction|GetHasGrantedTraining)) (Math|Float|InRange(Float) (Transformation|GetHorizontalDistanceTo (Variables|GetInteractingPawn) (Interaction|GetInterationDistanceCheckActor)) (ScalableFloat|GetValueatLevel (Variables|Interaction|Hotfixes|GetRow_MinDistanceBetweenForInteract)) (ScalableFloat|GetValueatLevel (Variables|Interaction|Hotfixes|GetRow_MaxDistanceBetweenForInteract)))))
    (else
      (return false))))

(fn BlueprintGetInteractionString
  (params (InteractingPawn :type Fort%20Pawn%20Object%20Reference) (InteractionBeingAttempted :type EInteractionBeingAttempted%20Enum :default "FirstInteraction"))
  (return ""))

(fn BlueprintGetInteractionTime
  (params (InteractingPawn :type Fort%20Pawn%20Object%20Reference) (InteractionBeingAttempted :type EInteractionBeingAttempted%20Enum :default "FirstInteraction"))
  (return 0.0 true))

(fn BlueprintOnLocalInteract
  (params (InteractingPawn :type Fort%20Player%20Pawn%20Object%20Reference) (InteractionBeingAttempted :type EInteractionBeingAttempted%20Enum :default "FirstInteraction"))
  (bind _returnvalue (|CallParentFunction InteractingPawn InteractionBeingAttempted))
  (if (GameplayTags|HasMatchingGameplayTag InteractingPawn (Variables|Interaction|GetTag_HasForcePower))
    (CallFunction|ShowDenyTrainingMessage InteractingPawn) ; display the deny message
    (else
      _))
  (return _returnvalue))

(fn BlueprintUpdateDynamicInteractionOptions
  (params (InteractingPawn :type Fort%20Pawn%20Object%20Reference) (InteractionBeingAttempted :type EInteractionBeingAttempted%20Enum :default "FirstInteraction") (InteractionType :type TInteractionType%20Enum :default "IT_NoInteraction"))
  (bind _outgameplaytags (|CallParentFunction InteractingPawn InteractionBeingAttempted InteractionType))
  (LocalVariables|SethOutInteractionOptions _outgameplaytags)
  (Utilities|FlowControl|Sequence
    (:then_0
      (if (GameplayTags|HasMatchingGameplayTag InteractingPawn (Variables|Interaction|GetTag_HasForcePower))
        (GameplayTags|AddGameplayTag (LocalVariables|GethOutInteractionOptions) "(TagName=\"HUD.Athena.InteractionOptions.DisableInteractKey\")" :!class /Script/GameplayTags.BlueprintGameplayTagLibrary) ; Disable interact key on UI message for unassigned players
      ))
    (:then_1
      (return (LocalVariables|GethOutInteractionOptions)))))

(fn GetClosestValidLookAtPawn
  (params (LocationToCheck :type Vector) (ActorsToIgnore :type Array%20of%20Actor%20Object%20References))
  (bind _self self)
  (bind _outactors (Collision|SphereOverlapActors LocationToCheck (ScalableFloat|GetValueatLevel (Variables|LookAtPlayer|Hotfixes|GetRow_FacePawn_MaxDistance)) (Utilities|Array|MakeArray "ObjectTypeQuery3") 0 ActorsToIgnore))
  (for _array_element _outactors
    ;[0.346 0.601 1] Should return the closest valid player pawn. Bot Controlled pawns are excluded and only the closest pawn should be stored. Keep the distance to check against if there is more than one pawn
    (bind _asfort_player_pawn (Utilities|Casting|CastToFortPlayerPawn _array_element))
    (if (not (Pawn|IsBotControlled _asfort_player_pawn))
      (LocalVariables|Geth_BestLookAtPawn
        (:then
          (if (>= (LocalVariables|Geth_BestPawnDistance) (Transformation|GetDistanceTo _asfort_player_pawn _self))
            _))
        (:else
          _))
      (LocalVariables|Seth_BestLookAtPawn _asfort_player_pawn)
      (LocalVariables|Seth_BestPawnDistance (Transformation|GetDistanceTo _asfort_player_pawn _self))))
  ;[1 0.4633 0.9643] RETURN
  ; closest valid pawn by distance and not a bot
  (return (LocalVariables|Geth_BestLookAtPawn)))

(fn GetInterationDistanceCheckActor
  (bind _associatedriftgate (Variables|Default|GetAssociatedRiftGate)
    (:then
      (return _associatedriftgate))
    (:else
      (return self))))

(fn GetLookatYawValue
  (params (ActorToLookAt :type Actor%20Object%20Reference))
  (bind (_returnvalue_roll _returnvalue_pitch _returnvalue_yaw) (Math|Rotator|BreakRotator (Math|Rotator|FindLookatRotation (Transformation|GetWorldLocation (Variables|Default|GetLookAtPawnCollisionBox)) (Transformation|GetActorLocation ActorToLookAt))))
  (return _returnvalue_yaw))

(fn GetPawnsForceTrainingComparison
  (params (FortPlayerPawnToTest :type Fort%20Player%20Pawn%20Object%20Reference))
  (Utilities|IsValid FortPlayerPawnToTest
    (:Is%20Valid
      (Utilities|FlowControl|Sequence
        (:then_0
          (if (not (GameplayTags|HasMatchingGameplayTag (Variables|GetFortPlayerPawnToTest) (Variables|Interaction|GetTag_HasForcePower)))
            (return "NewEnumerator0")))
        (:then_1
          (if (GameplayTags|HasMatchingGameplayTag (Variables|GetFortPlayerPawnToTest) (Variables|Dialogue|Affiliation|GetTag_ForcePowerGranted))
            (return "NewEnumerator1")))
        (:then_2
          (return (select (== (GameplayTags|HasAnyMatchingGameplayTags (Variables|GetFortPlayerPawnToTest) (Variables|Dialogue|Affiliation|GetTags_JediForcePowers)) (Variables|Dialogue|Affiliation|GetbIsThisActorJedi)) "NewEnumerator2" "NewEnumerator3")))))
    (:Is%20Not%20Valid
      (return "NewEnumerator0"))))

(fn GetProximityMessage
  (params (FortPlayerPawnToTest :type Fort%20Player%20Pawn%20Object%20Reference))
  (return (Utilities|Select :Index (Dialogue|Affiliation|GetPawnsForceTrainingComparison FortPlayerPawnToTest) :NewEnumerator0 (select (Utilities|Array|IsNotEmpty (Variables|Dialogue|GetDoesNotHaveForcePowerMessages)) (Utilities|Array|RandomArrayItem (Variables|Dialogue|GetDoesNotHaveForcePowerMessages)) "") :NewEnumerator1 (Variables|Dialogue|GetProximityMessageWasTrainedByThisActor) :NewEnumerator2 (Variables|Dialogue|GetProximityMessageWasTrainedBySameOrder) :NewEnumerator3 (Variables|Dialogue|GetProximityMessageWasTrainedByDifferentOrder))))

(fn OnRep_HasGrantedTraining
  (Networking|CanExecuteCosmeticEvents
    (:True
      ; Contextual Tutorial
      (if (TouchPlatform|IsMobileGame)
        (Ability|SendGameplayEventtoActor (Game|GetPlayerPawn 0) "(TagName=\"Event.Athena.PlotTwist.ContextualTutorial.HasGrantedTraining\")")))))

(fn UpdateCachedProximityMessage
  (params (PawnToCheck :type Fort%20Player%20Pawn%20Object%20Reference))
  (Utilities|FlowControl|Sequence
    (:then_0
      (Networking|CanExecuteCosmeticEvents
        (:True
          (if (ScalableFloat|GetValueasBool (Variables|Dialogue|Hotfixes|GetRow_ProximityMessages_Enabled))
            (Utilities|IsValid PawnToCheck
              (:Is%20Valid
                (if (Pawn|IsLocallyControlled PawnToCheck)
                  (bind _output_get (Variables|Dialogue|System|SetCachedProximityMessageRecipient PawnToCheck))
                  (bind _output_get_2 (Variables|Dialogue|System|SetCachedProximityMessage (Dialogue|GetProximityMessage _output_get)))
                  (return (not (Utilities|Text|TextIsEmpty _output_get_2))))))))))
    (:then_1
      ;[1 0.469 0.9125] Return FALSE - will run for all ended exec nodes
      (return false))))

(fn ConstructionScript)
