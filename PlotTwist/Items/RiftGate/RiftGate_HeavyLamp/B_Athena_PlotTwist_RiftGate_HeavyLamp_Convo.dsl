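; Rift Gate variant for the HeavyLamp NPC's conversation, sharing that NPC's location with hologram visuals off and granting training powers.
; On authority it skips the hologram spawn sequence and turns the NPC pawn, rather than itself, toward the trainee or closest valid pawn.
; When powers are granted, it moves the gate off the NPC using backward and sideways offsets, choosing the side from the trainee's position.
; It also excludes the NPC from interaction line-of-sight and signals when the trainee hides.
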(event EventPlaySpawnSequence
  ;[0.2855 1 0.3931] Event PlaySpawnSequence (OVERRIDE)
  ; This rift should skip the spawning in visuals because the hologram isn't seen
  (Networking|SwitchHasAuthority
    (:Authority
      (CallFunction|SpawnIn) ; skips the hologram spawning in sequence \nbecause there is already an owning pawn in the location\nand hologram visuals are disabled
    )))

(event EventFaceInteractingOrClosestPawn
  (bind _row_facepawn_rotationperupdate (Variables|FacePawn|GetRow_FacePawn_RotationPerUpdate))
  (bind _returnvalue (ScalableFloat|GetValueatLevel _row_facepawn_rotationperupdate))
  ;[0.4656 0.6228 1] Event FaceInteractingOrClosestPawn (OVERRIDE)
  ; Ignores the OwningNPC (HeavyLamp) and turns the Heavy Lamp pawn instead of the hologram
  (bind _beinggrantedpowerspawn (Variables|RiftGate|Orientation|GetBeingGrantedPowersPawn)
    (:then
      (Variables|FacePawn|SetHologramsFocusedPawn _beinggrantedpowerspawn))
    (:else
      (if (not (Variables|Interaction|ControlVars|GetHasGrantedTraining))
        (bind _closestvalidlookatpawn (Hologram|GetClosestValidLookatPawn (Transformation|GetActorLocation) (Utilities|Array|MakeArray (Variables|Default|GetOwningNPCPawn))))
        (Utilities|IsValid _closestvalidlookatpawn
          (:Is%20Valid
            (Variables|FacePawn|SetHologramsFocusedPawn _closestvalidlookatpawn))) ; can return null
      )))
  (bind _owningnpcpawn (Variables|Default|GetOwningNPCPawn)
    (:then
      (bind _hologramsfocusedpawn (Variables|FacePawn|GetHologramsFocusedPawn)
        (:then
          (bind _output_get (Variables|FacePawn|SetDesiredLookatYaw (Hologram|GetLookatYawValue _hologramsfocusedpawn)))
          (bind (_returnvalue_roll _returnvalue_pitch _returnvalue_yaw) (Math|Rotator|BreakRotator (Transformation|GetWorldRotation (Variables|/Script/Engine.Actor|GetRootComponent _owningnpcpawn))))
          (if (not (Math|Float|NearlyEqual(Float) _output_get _returnvalue_yaw (* _returnvalue 1.5)))
            (bind (_returnvalue_roll_2 _returnvalue_pitch_2 _returnvalue_yaw_2) (Math|Rotator|BreakRotator (Transformation|GetWorldRotation (Variables|/Script/Engine.Actor|GetRootComponent _owningnpcpawn))))
            (bind (_returnvalue_roll_3 _returnvalue_pitch_3 _returnvalue_yaw_3) (Math|Rotator|BreakRotator (Math|Interpolation|RInterpTo (Math|Rotator|MakeRotator 0.0 0.0 _returnvalue_yaw_2) (Math|Rotator|MakeRotator 0.0 0.0 _output_get) (Math|Float|SafeDivide 1.0 (ScalableFloat|GetValueatLevel (Variables|FacePawn|GetRow_FacePawn_UpdateRate))) _returnvalue)))
            (Transformation|SetWorldRotation (Variables|/Script/Engine.Actor|GetRootComponent _owningnpcpawn) (Math|Rotator|MakeRotator 0.0 0.0 _returnvalue_yaw_3) false false)))
        (:else)))
    (:else)))

(event EventGainForcePowers
  ;[0.4656 0.6228 1] Event GainForcePowers
  ; Extending function to move the Rift Gate before it is visually spawned so that it is not in the exact same position as the OwningNPC (HeavyLamp)
  ;;[0.4656 0.6228 1] When training is accepted, move the rift gate away from the OwningNPClocation, ideally to the side and slightly behind
  (bind _beinggrantedpowerspawn (Variables|RiftGate|Orientation|GetBeingGrantedPowersPawn)
    (:then
      (bind _owningnpcpawn (Variables|Default|GetOwningNPCPawn)
        (:then
          (Transformation|SetActorLocation (Utilities|Operators|Add (Transformation|GetActorLocation _owningnpcpawn) (* (Transformation|GetActorForwardVector _owningnpcpawn) (Variables|Default|GetOffsetRiftPortalBackwards)) (* (Transformation|GetActorRightVector _owningnpcpawn) (Math|Float|SelectFloat (* (Variables|Default|GetOffsetRiftSideways) -1.0) (Variables|Default|GetOffsetRiftSideways) (> (Math|Vector|DotProduct (- (Transformation|GetActorLocation _owningnpcpawn) (Transformation|GetActorLocation _beinggrantedpowerspawn)) (Transformation|GetActorRightVector _owningnpcpawn)) 0))))))
        (:else
          _))
      (|CallParentFunction))
    (:else
      _)))

(event EventStartHiding
  (params (InteractingPawn :type Fort%20Pawn%20Object%20Reference))
  (|CallParentFunction InteractingPawn)
  (bind _beinggrantedpowerspawn (Variables|RiftGate|Orientation|GetBeingGrantedPowersPawn)
    (:then
      (if (== _beinggrantedpowerspawn InteractingPawn)
        (Default|CallEventonGrantedPowersPawnIsHiding)))
    (:else)))

(fn GetIgnoredActorsFromLineOfSight
  (params (IgnoredLOSActors :type Array%20of%20Actor%20Object%20References) (InteractingPawn :type Actor%20Object%20Reference))
  ;[0.2363 0.4615 0.7708] Add the owning NPC to the ignored actor array so it doesn't block interaction Line of Sight and they inhabit the same physical location as this Rift Gate
  (bind _output_get (LocalVariables|Seth_CombinedActorsToIgnore (|CallParentFunction IgnoredLOSActors InteractingPawn)))
  (Utilities|Array|AddUnique _output_get (Variables|Default|GetOwningNPCPawn))
  (return _output_get))

(fn GetLookatYawValue
  (params (ActorToLookat :type Actor%20Object%20Reference))
  (bind _owningnpcpawn (Variables|Default|GetOwningNPCPawn)
    (:then
      (bind (_returnvalue_roll _returnvalue_pitch _returnvalue_yaw) (Math|Rotator|BreakRotator (Math|Rotator|FindLookatRotation (Transformation|GetActorLocation _owningnpcpawn) (Transformation|GetActorLocation ActorToLookat))))
      (return _returnvalue_yaw))
    (:else
      (return 0.0))))

(fn ConstructionScript
  (|CallParentFunction))
