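; Open-rift variant of the HeavyLamp Rift Gate that skips the parent's hologram spawn-in, opens immediately, shows a map icon and pings its location.
; Server-side hits knock back pawns other than the owning NPC and trainee via gameplay effects, with optional fall-damage immunity.
; An overlap zone launches vehicles outward, with knockback behavior hotfixable through scalable floats.
; Safe-zone checks start a storm timer, then trigger forced or delayed visual despawn based on nearby pawns.
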(event EventBeginPlay
  ;[0 1 0.02588] BeginPlay
  ; Setup visual state based on rift gate status, allow idle collision to be disabled via hotfix
  (Utilities|FlowControl|Sequence
    (:then_0
      ;;[0.2904 0.0144 0.4844] Activate Map Icon
      (Components|Activation|Activate (Variables|Default|GetAthenaSpecialActor) false) ; must run on server + clients
      (Networking|CanExecuteCosmeticEvents
        (:True
          (Map|TryAddSpecialActorRenderTag false) ; mini map icon tag
          (switch Utilities|FlowControl|Switch|SwitchonEnum_RiftGateState (Variables|RiftGate|State|GetRiftGateState_Current)
            (:NewEnumerator0
              _)
            (:NewEnumerator1
              _)
            (:NewEnumerator2
              (|CallParentFunction :!label _label) ; run default logic
            )
            (:NewEnumerator3
              (goto _label)))
          ;;[0.1773 1 0.4438] Force the correct visual state for opened rifts
          (Rendering|SetActorHiddenInGame false) ; FORT-886506\nthe parent begin play will hide and\n respawn the force trainer every time, but \nbecause we are in the open state already \nand skip the animations, it will disable \nthe Rift Gate visuals if not handled
          (Development|SetHiddeninGame (Variables|Default|GetSM_Pickle_Port_Inner) false true)
          (CallFunction|SetUIMarkerVisibility false)
          (CallFunction|CheckforPawninArea) ; look at functionality
        )
        (:False
          _))
      (goto _label))
    (:then_1
      (if (not (ScalableFloat|GetValueasBool (Variables|Knockback|Hotfixes|GetHFRow_bEnableOpenRiftCollision)))
        (Collision|SetCollisionEnabled (Variables|Default|GetIdleKnockbackArea) "NoCollision") ; disable collision on the Rift Gate when set via data row (hotfixable)
      ))))

(event EventSpawnIn
  ;[0.1599 1 0.3125] Event SpawnIn (OVERRIDE)
  ; The parent rift gate will spawn in the hologram, but for this version we want to immediately spawn the open rift
  (Components|Activation|Activate (Variables|Default|GetAthenaSpecialActor) false) ; show on map
  (Rendering|SetActorHiddenInGame false) ; the rift gate starts in a "collapsed" \nstate, so this will look fine
  (CallFunction|SetRiftGateOpen true) ; animates open
  (Utilities|Time|SetTimerbyEvent (EventDispatchers|CreateEvent :!event Custom|MarkRiftGateOnOpened) 1.1) ; delay the UI mark indicator for a moment to give\nthe gate to appear visually
)

(event Custom|MarkRiftGateOnOpened
  (bind _returnvalue (Fortnite|GetFirstLocalFortPlayerController true))
  ;[0.1473 0.000377 0.3411] MarkriftGateOnOpened
  ; use a ping to show the Rift Gate location
  (Utilities|IsValid _returnvalue
    (:Is%20Valid
      (Fortnite|RegisterSoundIndicatorPlacementContext _returnvalue self "0, 0, 0" false))))

(event Custom|StartDelayToVisualDespawn
  (bind _hfrow_delayedtimeuntildespawn (Variables|SafeZone|GetHFRow_DelayedTimeUntilDespawn))
  (bind _returnvalue (ScalableFloat|GetValueatLevel _hfrow_delayedtimeuntildespawn))
  ;[0.1473 0.000377 0.3411] StartDelayToVisualDespawn
  ; Give Owning NPC and other actors time to "exit" through the rift gate
  (if (> _returnvalue 0)
    (Utilities|Time|SetTimerbyEvent (EventDispatchers|CreateEvent :!event Custom|VisualNonTrainedDespawn) _returnvalue)
    (else
      (CallFunction|VisualNonTrainedDespawn) ; delay is zero or less, do it now
    )))

(event Custom|VisualNonTrainedDespawn
  ;[0.1473 0.000377 0.3411] VisualNonTrainedDespawn
  ; visually close the Rift Gate and then despawn the Rift Gate
  (CallFunction|SetRiftGateOpen false)
  (Actor|SetLifeSpan 2.0))

(event OnComponentHit(IdleKnockbackArea)
  (params (HitComponent :type Primitive%20Component%20Object%20Reference) (OtherActor :type Actor%20Object%20Reference) (OtherComp :type Primitive%20Component%20Object%20Reference) (NormalImpulse :type Vector :default "0, 0, 0") (Hit :type Hit%20Result%20Structure%20%28by%20ref%29))
  (bind _self self)
  (bind _returnvalue (Ability|GetAbilitySystemComponent _self))
  ;[0.375 0.1892 0.1477] OnComponentHit
  ; Try to knockback players from the rift area to keep it clear and limit players ability from camping inside the visuals
  (Networking|SwitchHasAuthority
    (:Authority
      (bind _asfort_player_pawn_athena (Utilities|Casting|CastToFortPlayerPawnAthena OtherActor))
      (if (and (Knockback|ShouldKnockbackCollidingPawn _asfort_player_pawn_athena) (ScalableFloat|GetValueasBool (Variables|Knockback|Hotfixes|GetHFRow_bEnableIdleKnockback)))
        (Utilities|IsValid _returnvalue
          (:Is%20Valid
            (GameplayEffects|ApplyGameplayEffectToTarget _returnvalue (Variables|Knockback|GetGE_IdleKnockback) (Ability|GetAbilitySystemComponent _asfort_player_pawn_athena) 0.0 (Class|AthenaFunctionLibrary|MakeSimpleFortEffectContext self self)) ; Apply knockback effect
            ;;[0.6574 0.1718 0.8594 0.5] Play knockback sound on player when player bumps into rift
            (GameplayCue|ExecuteGameplayCueOnActor(Burst) _asfort_player_pawn_athena (Variables|Knockback|GetBounceBack%20Gameplay%20Cue%20Tag) (Ability|GameplayCue|MakeGameplayCueParameters 0.0))
            (if (ScalableFloat|GetValueasBool (Variables|Knockback|Hotfixes|GetHFRow_bEnableIdleKnockbackFallDamageImmunity))
              (GameplayEffects|ApplyGameplayEffectToTarget :self (Ability|GetAbilitySystemComponent _asfort_player_pawn_athena) :GameplayEffectClass (Variables|Knockback|GetGE_FallDamageImmunity) :Target (Ability|GetAbilitySystemComponent _asfort_player_pawn_athena) :Level 0.0) ; fall damage immunity
            )))))) ; run on server, knockback GA will replicate to clients
)

(event EventStopHiding
  (params (Pawn :type Fort%20Pawn%20Object%20Reference))
  (|CallParentFunction Pawn)
  (Networking|SwitchHasAuthority
    (:Authority
      (Default|CallOnRiftGateSequenceFinished))))

(event OnComponentBeginOverlap(VehicleKnockbackZone)
  (params (OverlappedComponent :type Primitive%20Component%20Object%20Reference) (OtherActor :type Actor%20Object%20Reference) (OtherComp :type Primitive%20Component%20Object%20Reference) (OtherBodyIndex :type Integer) (bFromSweep :type Boolean) (SweepResult :type Hit%20Result%20Structure%20%28by%20ref%29))
  (bind _vehicleknockbackzone (Variables|Default|GetVehicleKnockbackZone))
  ;[0.375 0.1892 0.1477] OnComponentBeginOverlap (VehicleKnockbackZone)
  ; Check knockback and use rift gate sequence values to knock the vehicle away from the Rift Gate and the owning NPC
  (if (ScalableFloat|GetValueasBool (Variables|Knockback|Hotfixes|GetVehicleIdlePushbackEnabled))
    (bind _asfort_vehicle_interface (Utilities|Casting|CastToFortVehicleInterface OtherActor)) ; should only be a vehicle, but confirm
    (bind _returnvalue (Vehicle|GetVehicleActor _asfort_vehicle_interface :!class /Script/FortniteGame.FortVehicleInterface))
    (Utilities|IsValid _returnvalue
      (:Is%20Valid
        (Audio|BroadcastAbilityImpactatLocation 0 _returnvalue (Knockback|CalculateVehicleKnockbackVelocity _returnvalue) 777.0) ; vehicle knockback radius MUST be 777
        (GameplayCue|ExecuteGameplayCueOnActor(Burst) _returnvalue (Variables|Knockback|GetGameplayCueForVehicleKnockback) (GameplayCue|MakeGameplayCueParametersfromHitResult SweepResult)) ; standard rift gate knockback from CinematicSequence
      ))
    (else
      ;;[1 0.3247 0] Since this component is only for tracking vehicle collision, we can disable it if the hotfix is set to not have knockback on vehicle collision
      (Collision|SetCollisionEnabled _vehicleknockbackzone "NoCollision")
      (Components|Activation|Deactivate _vehicleknockbackzone))))

(fn CalculateVehicleKnockbackVelocity
  (params (VehicleToKnockback :type Actor%20Object%20Reference))
  (bind (_returnvalue_x _returnvalue_y) (Math|Vector|BreakVector (Math|Vector|Normalize2D(Vector) (- (Transformation|GetActorLocation VehicleToKnockback) (Transformation|GetActorLocation self)))))
  (return (Math|Vector|MakeVector (* _returnvalue_x (ScalableFloat|GetValueatLevel (Variables|Knockback|Hotfixes|GetVehicleHorizontalVelocity))) (* _returnvalue_y (ScalableFloat|GetValueatLevel (Variables|Knockback|Hotfixes|GetVehicleHorizontalVelocity))) (ScalableFloat|GetValueatLevel (Variables|Knockback|Hotfixes|GetHFRow_VehicleVerticalVelocity)))))

(fn CanPawnsInStormStopDespawn
  (bind _outfortplayerpawns (Fortnite|GetFortPlayerPawnsInRange (Transformation|GetActorLocation) (ScalableFloat|GetValueatLevel (Variables|SafeZone|GetHFRow_InStormDistanceToValidPawns)) true)) ; Are there any pawns close?\nThis should handle conversation and cinematic sequences
  (bind _owningnpcpawn (Variables|Default|GetOwningNPCPawn)
    (:then
      (bind _array_element (Utilities|Array|ForEachLoopwithBreak _outfortplayerpawns :!label _label)
        (:LoopBody
          (if (not (== _array_element _owningnpcpawn))
            (LocalVariables|Seth_FoundPawnPreventingDespawn true)
            (goto _label Break)))
        (:Completed
          _)))
    (:else
      _))
  (return (or (not (LocalVariables|Geth_FoundPawnPreventingDespawn)) (Utilities|Array|IsEmpty _outfortplayerpawns))))

(fn CheckForInSafeZone
  (Networking|SwitchHasAuthority
    (:Authority
      ;[1 0.3042 0.1474] If we passed Check 1
      ; Skip and go to Check 2
      (if (> (Variables|SafeZone|GetOutOfSafeZoneDestroyTime) 0)
        ;[0.1708 1 0.2319] Check 2
        (if (>= (Fortnite|Sync|Time|GetServerWorldTimeSeconds) (Variables|SafeZone|GetOutOfSafeZoneDestroyTime))
          ;[0.1708 1 0.2319] Check 3
          (bind _nopawnpreventingdespawn (SafeZone|CanPawnsinStormStopDespawn))
          (if _nopawnpreventingdespawn
            (Default|CallOnRiftGateForcedDespawn)
            (CallFunction|StartDelaytoVisualDespawn)))
        (else
          ;[0.1708 1 0.2319] Check 1
          (bind _returnvalue (SafeZone|IsLocationinSafeZone (Transformation|GetActorLocation)))
          (if (not _returnvalue)
            (Variables|SafeZone|SetOutOfSafeZoneDestroyTime (+ (Fortnite|Sync|Time|GetServerWorldTimeSeconds) (Variables|SafeZone|GetOutOfSafeZoneLifetime)))))))))

(fn GetIgnoredActorsFromLineOfSight
  (params (IgnoredLOSActors :type Array%20of%20Actor%20Object%20References) (InteractingPawn :type Actor%20Object%20Reference))
  ;[0.2363 0.4615 0.7708] Add the owning NPC to the ignored actor array so it doesn't block interaction Line of Sight and they inhabit the same physical location as this Rift Gate
  (bind _output_get (LocalVariables|Seth_CombinedActorsToIgnore (|CallParentFunction IgnoredLOSActors InteractingPawn)))
  (Utilities|Array|AddUnique _output_get (Variables|Default|GetOwningNPCPawn))
  (return _output_get))

(fn GetMarkerDetailsId
  (return "MarkerDetails.RiftGate.HeavyLamp"))

(fn RotateHideLocationVectorAndPortalToTarget
  (params (TargetActor :type Actor%20Object%20Reference)))

(fn ShouldKnockbackCollidingPawn
  (params (ActorToCheck :type Actor%20Object%20Reference))
  (bind _owningnpcpawn (Variables|Default|GetOwningNPCPawn)
    (:then
      (if (== _owningnpcpawn ActorToCheck)
        ;[1 0.4922 0.9954] Pawns not being trained and not the owning NPC can be knocked back
        (return false)
        (else
          _)))
    (:else
      _))
  (bind _beinggrantedpowerspawn (Variables|RiftGate|Orientation|GetBeingGrantedPowersPawn)
    (:then
      (if (== _beinggrantedpowerspawn ActorToCheck)
        (return false)
        (else
          _)))
    (:else
      _)) ; DONT knockback player being trained
  (return true))

(fn ConstructionScript
  (|CallParentFunction))
